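; Conversation requirement for the Clyde Adventure Girl boss NPC that gates dialogue by week.
; IsRequirementSatisfied gets the participant NPC from the conversation context, casts it to the boss pawn and compares its ClydeWeekIndex to the ValidWeek variable.
; A match returns Passed, while any other week returns FailedAndHidden.
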(fn IsRequirementSatisfied
  (params (Context :type Conversation%20Context%20Structure%20%28by%20ref%29))
  (bind _returnvalue (Class|FortConversationContextLibrary|GetParticipantActorNPC Context))
  (bind _asbp_player_pawn_clyde_adv_girl_boss (Utilities|Casting|CastToBP_PlayerPawn_Clyde_AdvGirlBoss _returnvalue))
  (if (== (Variables|/Clyde/Athena/AI/Clyde/Bosses/AdventureGirl/BP_PlayerPawn_Clyde_AdvGirlBoss.BP_PlayerPawn_Clyde_AdvGirlBoss_C|GetClydeWeekIndex _asbp_player_pawn_clyde_adv_girl_boss) (Variables|Default|GetValidWeek))
    (return "Passed")
    (else
      (return "FailedAndHidden"))))
