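# Conversation database for Skye, the Clyde Adventure Girl boss NPC, giving her flavor dialogue that changes by week.
# The entry point fans out to four ValidateWeek task nodes, one at the default week and the others for weeks 1 to 3.
# Each leads to an "Ask about..." greeting with two question-and-reply topics and a Back choice.
# Replies exit through ReturnToStart, except one topic each in weeks 2 and 3 that uses Back.
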ConversationDatabase /Clyde/Athena/AI/Clyde/Bosses/AdventureGirl/NPCLibrary_Conversation_Clyde_AdvGirlBoss
entry_points [{tag=Conversation.Entry.Clyde.AdvGirl, destinations [#1]}]
nodes [
 ConversationEntryPointNode_1 {
  id=#1
  class=ConversationEntryPointNode
  outputs [#2, #3, #4, #5]
  properties {entryTag {tagName=Conversation.Entry.Clyde.AdvGirl}}
 }
 FCTN_Clyde_GetWeek_C_3 {id=#5, class=FCTN_Clyde_ValidateWeek_C, outputs [#6], properties {validWeek=3}}
 Speech_C_7 {
  id=#6
  class=Speech_C
  outputs [#7, #8, #9]
  properties {nodeName="", softTaskIcon=&T_UI_NPC_question, generalConfig {defaultMessage="Adventure calls!"}}
  sub_nodes [ConversationChoiceNode_4 {class=ConversationChoiceNode, properties {defaultChoiceDisplayText=Ask about...}}]
 }
 Speech_C_11 {
  id=#9
  class=Speech_C
  outputs [#10]
  properties {
   nodeName=""
   softTaskIcon=&T_UI_NPC_question
   generalConfig {defaultMessage="Ooh, I've gotta burn off this excited energy!! Maybe some light pacing'll do the trick..."}
  }
  sub_nodes [ConversationChoiceNode_5 {class=ConversationChoiceNode, properties {defaultChoiceDisplayText=What do we do now?}}]
 }
 ReturnToStart_C_4 {
  id=#10
  class=ReturnToStart_C
  properties {softTaskIcon=&T_UI_NPC_Back}
  sub_nodes [ConversationChoiceNode_6 {class=ConversationChoiceNode, properties {defaultChoiceDisplayText=BACK}}]
 }
 Back_C_1 {
  id=#8
  class=Back_C
  sub_nodes [ConversationChoiceNode_21 {class=ConversationChoiceNode, properties {defaultChoiceDisplayText=BACK}}]
 }
 Speech_C_5 {
  id=#7
  class=Speech_C
  outputs [#11]
  properties {
   nodeName=""
   softTaskIcon=&T_UI_NPC_question
   generalConfig {defaultMessage="Have you heard?! Something big is coming!!"}
  }
  sub_nodes [ConversationChoiceNode_8 {class=ConversationChoiceNode, properties {defaultChoiceDisplayText="What's up, Skye?"}}]
 }
 Back_C_5 {
  id=#11
  class=Back_C
  sub_nodes [ConversationChoiceNode_7 {class=ConversationChoiceNode, properties {defaultChoiceDisplayText=BACK}}]
 }
 FCTN_Clyde_GetWeek_C_2 {id=#4, class=FCTN_Clyde_ValidateWeek_C, outputs [#12], properties {validWeek=2}}
 Speech_C_2 {
  id=#12
  class=Speech_C
  outputs [#13, #14, #15]
  properties {nodeName="", softTaskIcon=&T_UI_NPC_question, generalConfig {defaultMessage="Adventure calls!"}}
  sub_nodes [ConversationChoiceNode_17 {class=ConversationChoiceNode, properties {defaultChoiceDisplayText=Ask about...}}]
 }
 Speech_C_12 {
  id=#15
  class=Speech_C
  outputs [#16]
  properties {
   nodeName=""
   softTaskIcon=&T_UI_NPC_question
   generalConfig {defaultMessage=Come back later. The show in the sky will be legendary.}
  }
  sub_nodes [ConversationChoiceNode_26 {class=ConversationChoiceNode, properties {defaultChoiceDisplayText=What do we do now?}}]
 }
 ReturnToStart_C_5 {
  id=#16
  class=ReturnToStart_C
  properties {softTaskIcon=&T_UI_NPC_Back}
  sub_nodes [ConversationChoiceNode_27 {class=ConversationChoiceNode, properties {defaultChoiceDisplayText=BACK}}]
 }
 Back_C_0 {
  id=#14
  class=Back_C
  sub_nodes [ConversationChoiceNode_20 {class=ConversationChoiceNode, properties {defaultChoiceDisplayText=BACK}}]
 }
 Speech_C_1 {
  id=#13
  class=Speech_C
  outputs [#17]
  properties {
   nodeName=""
   softTaskIcon=&T_UI_NPC_question
   generalConfig {defaultMessage="I've heard the sunset is beautiful out here. "}
  }
  sub_nodes [ConversationChoiceNode_16 {class=ConversationChoiceNode, properties {defaultChoiceDisplayText=What are you looking for?}}]
 }
 Back_C_4 {
  id=#17
  class=Back_C
  sub_nodes [ConversationChoiceNode_18 {class=ConversationChoiceNode, properties {defaultChoiceDisplayText=BACK}}]
 }
 FCTN_Clyde_GetWeek_C_1 {id=#3, class=FCTN_Clyde_ValidateWeek_C, outputs [#18], properties {validWeek=1}}
 Speech_C_0 {
  id=#18
  class=Speech_C
  outputs [#19, #20, #21]
  properties {nodeName="", softTaskIcon=&T_UI_NPC_question, generalConfig {defaultMessage="Adventure calls!"}}
  sub_nodes [ConversationChoiceNode_11 {class=ConversationChoiceNode, properties {defaultChoiceDisplayText=Ask about...}}]
 }
 Speech_C_4 {
  id=#21
  class=Speech_C
  outputs [#22]
  properties {
   nodeName=""
   softTaskIcon=&T_UI_NPC_question
   generalConfig {defaultMessage="FINS WHERE I CAN SEE 'EM-- oh, you're not the shark I'm looking for."}
  }
  sub_nodes [ConversationChoiceNode_10 {class=ConversationChoiceNode, properties {defaultChoiceDisplayText="Hi Skye!"}}]
 }
 ReturnToStart_C_0 {
  id=#22
  class=ReturnToStart_C
  properties {softTaskIcon=&T_UI_NPC_Back}
  sub_nodes [ConversationChoiceNode_19 {class=ConversationChoiceNode, properties {defaultChoiceDisplayText=BACK}}]
 }
 Back_C_3 {
  id=#20
  class=Back_C
  sub_nodes [ConversationChoiceNode_12 {class=ConversationChoiceNode, properties {defaultChoiceDisplayText=BACK}}]
 }
 Speech_C_3 {
  id=#19
  class=Speech_C
  outputs [#23]
  properties {
   nodeName=""
   softTaskIcon=&T_UI_NPC_question
   generalConfig {defaultMessage="Gear up and get outta here. And NO BITING!! I'll be watching you..."}
  }
  sub_nodes [ConversationChoiceNode_22 {class=ConversationChoiceNode, properties {defaultChoiceDisplayText=What do we do now?}}]
 }
 ReturnToStart_C_1 {
  id=#23
  class=ReturnToStart_C
  properties {softTaskIcon=&T_UI_NPC_Back}
  sub_nodes [ConversationChoiceNode_23 {class=ConversationChoiceNode, properties {defaultChoiceDisplayText=BACK}}]
 }
 FCTN_Clyde_GetWeek_C_0 {id=#2, class=FCTN_Clyde_ValidateWeek_C, outputs [#24]}
 Speech_C_9 {
  id=#24
  class=Speech_C
  outputs [#25, #26, #27]
  properties {nodeName="", softTaskIcon=&T_UI_NPC_question, generalConfig {defaultMessage="Adventure calls!"}}
  sub_nodes [ConversationChoiceNode_14 {class=ConversationChoiceNode, properties {defaultChoiceDisplayText=Ask about...}}]
 }
 Speech_C_10 {
  id=#27
  class=Speech_C
  outputs [#28]
  properties {
   nodeName=""
   softTaskIcon=&T_UI_NPC_question
   generalConfig {defaultMessage=If you haven't seen any spaghetti or chainsaws... all we can do is gear up and go.}
  }
  sub_nodes [ConversationChoiceNode_24 {class=ConversationChoiceNode, properties {defaultChoiceDisplayText=What do we do now?}}]
 }
 ReturnToStart_C_3 {
  id=#28
  class=ReturnToStart_C
  properties {softTaskIcon=&T_UI_NPC_Back}
  sub_nodes [ConversationChoiceNode_25 {class=ConversationChoiceNode, properties {defaultChoiceDisplayText=BACK}}]
 }
 Back_C_2 {
  id=#26
  class=Back_C
  sub_nodes [ConversationChoiceNode_15 {class=ConversationChoiceNode, properties {defaultChoiceDisplayText=BACK}}]
 }
 Speech_C_8 {
  id=#25
  class=Speech_C
  outputs [#29]
  properties {
   nodeName=""
   softTaskIcon=&T_UI_NPC_question
   generalConfig {defaultMessage=Um... is this where I pick up my free spaghetti with the purchase of any chainsaw?}
  }
  sub_nodes [ConversationChoiceNode_13 {class=ConversationChoiceNode, properties {defaultChoiceDisplayText=You look lost...}}]
 }
 ReturnToStart_C_2 {
  id=#29
  class=ReturnToStart_C
  properties {softTaskIcon=&T_UI_NPC_Back}
  sub_nodes [ConversationChoiceNode_9 {class=ConversationChoiceNode, properties {defaultChoiceDisplayText=BACK}}]
 }
]
